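# Translucent unlit material for Niagara mesh particles, rendering an electric energy strip effect for the BariumDemon pickaxe.
# A slowly panning cloud noise feeds a named reroute for noise-based distortion of the parameter-tiled, vertically panning electricity texture.
# That texture multiplies particle color into an emissive adjustment function.
# A smoothstep edge mask, offset by vertex-color green with position and width parameters, is multiplied by particle alpha.
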Material /BRCosmetics/Weapons/FORT_Melee/Pickaxe_BariumDemon/FX/M_Pickaxe_BariumDemon_EnergyStrip
graph {
 properties {
  bUsedWithNiagaraMeshParticles=true
  blendMode=BLEND_Translucent
  pixelDepthOffsetMode=PDOM_Legacy
  shadingModel=MSM_Unlit
  usedShadingModels=Unlit
 }
 nodes [
  n1 = ParticleColor()
  n2 = Multiply(A=NamedRerouteUsage(), B=ScalarParameter() {parameterName=Noise_Distortion})
  n3 = VectorParameter() {parameterName=TexTiling, defaultValue=<rgba 0 0 0 1>}
  n4 = Add(
   A=Multiply(A=TextureCoordinate() {uTiling=0.5}, B=ComponentMask(None=@n3.RGB) {g=true, r=true})
   B=MaterialFunctionCall(R=@n3.B, G=@n3.B) {materialFunction=&AppendMany}.RG
  )
  n5 = TextureSample(UVs=Panner(Coordinate=Add(A=@n2, B=@n4)) {speedY=1}) {texture=/Game/Accessories/FORT_Capes/Cape_M_MED_BariumDemon/FX/T_FX_BariumDemon_Electricity}
  Add(A=@n1.RGB, B=@n5.G)
  MaterialFunctionCall(Emissive=Multiply(A=@n1.RGB, B=@n5.G)) {materialFunction=&MF_FX_EmissiveAdjustments}
  n6 = Add(A=VertexColor().G, B=ScalarParameter() {parameterName=Edge_P})
  n7 = ScalarParameter() {parameterName=Edge_W}
  Multiply(A=SmoothStep(Min=@n6, Max=Add(A=@n6, B=@n7), Value=NamedRerouteUsage()), B=@n1.A)
  n8 = TextureCoordinate() {uTiling=0.25, vTiling=0.5}
  NamedRerouteDeclaration(None=TextureSample(UVs=Panner(Coordinate=@n8) {speedY=0.25}) {samplerType=SAMPLERTYPE_LinearGrayscale, texture=/Game/Effects/Fort_Effects/Emotes/UltraLight/Texture/T_Cloud_Greyscale}.G) {name=MainNoise, nodeColor=<rgba 1 0 0.1641 1>}
 ]
}
