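# Unlit, two-sided Niagara sprite material for the internal energy effect on the BariumDemon pickaxe.
# Builds three masks: panning electricity with center-scaled, distorted UVs, a panning cloud noise, and a soft spherical light source.
# Their weighted sum tints BrightColor, which is screen-blended over BaseColor and fed to emissive and contrast adjustments.
# Dynamic parameters let particles drive light radius and electricity brightness.
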Material /BRCosmetics/Weapons/FORT_Melee/Pickaxe_BariumDemon/FX/M_Pickaxe_BariumDemon_InternalEnergy
graph {
 properties {bUsedWithNiagaraSprites=true, pixelDepthOffsetMode=PDOM_Legacy, shadingModel=MSM_Unlit, twoSided=true, usedShadingModels=Unlit}
 nodes [
  Add()
  n1 = ScalarParameter() {parameterName=MainNoise_Distortion}
  ComponentMask(None=@n1) {r=true}
  n2 = Add(
   A=Multiply(A=NamedRerouteUsage(), B=DynamicParameter() {paramNames [LightRadius, ElectricityBrightness, Param3, Param4]}.ElectricityBrightness)
   B=NamedRerouteUsage()
  )
  n3 = Multiply(A=NamedRerouteUsage(), B=ScalarParameter() {parameterName=MainNoise_Brightness})
  n4 = MaterialFunctionCall!(
   Blend=Multiply(A=VectorParameter() {parameterName=BrightColor, defaultValue=<rgba 0 0 0 1>}.RGB, B=Add(A=@n2, B=@n3))
   Base=VectorParameter() {parameterName=BaseColor, defaultValue=<rgba 0 0 0 1>}.RGB
  ) {materialFunction=/Engine/Functions/Engine_MaterialFunctions03/Blends/Blend_Screen}
  MaterialFunctionCall(Emissive=@n4.Result) {materialFunction=&MF_FX_EmissiveAdjustments}
  MaterialFunctionCall!(Color=@n4.Result, Contrast=ScalarParameter() {defaultValue=1, parameterName=Contrast}) {materialFunction=/Engine/Functions/Engine_MaterialFunctions01/ImageAdjustment/Contrast_Preserve_Color}
  n5 = VectorParameter() {parameterName=ElecScale, defaultValue=<rgba 3 1 0 1>}
  n6 = Multiply(A=NamedRerouteUsage()) {constB=2}
  n7 = Add(
   A=MaterialFunctionCall(UVs=TextureCoordinate() {vTiling=1.5}, Texture Scale=ComponentMask(None=@n5.RGB) {g=true, r=true}) {materialFunction=&ScaleUVsByCenter}.UVs
   B=Multiply(A=Subtract(A=@n6), B=@n1)
  )
  n8 = AppendVector(A=@n5.B, B=@n5.B)
  NamedRerouteDeclaration(None=TextureSample(UVs=Panner(Coordinate=@n7, Speed=@n8) {bFractionalPart=true, speedY=6}) {texture=/Game/Accessories/FORT_Capes/Cape_M_MED_BariumDemon/FX/T_FX_BariumDemon_Electricity}.R) {name=MainElectricity, nodeColor=<rgba 0 1 0.0391 1>}
  n9 = SphereMask(
   A=TextureCoordinate()
   B=Constant2Vector() {g=1, r=0.5}
   Radius=DynamicParameter() {paramNames [LightRadius, ElectricityBrightness, Param3, Param4]}.LightRadius
  ) {attenuationRadius=0.75, hardnessPercent=0}
  NamedRerouteDeclaration(None=Multiply(A=Power(Base=@n9), B=ScalarParameter() {parameterName=LightBrightness})) {name=LightSource, nodeColor=<rgba 0.0547 1 0 1>}
  n10 = Multiply(A=TextureCoordinate(), B=ComponentMask(None=VectorParameter() {parameterName=Noise_Tile, defaultValue=<rgba 1 1 0 1>}.RGB) {g=true, r=true})
  n11 = ComponentMask(None=VectorParameter() {parameterName=Noise_Speed, defaultValue=<rgba 1 1 0 1>}.RGB) {g=true, r=true}
  NamedRerouteDeclaration(None=TextureSample(UVs=Panner(Coordinate=@n10, Speed=@n11) {bFractionalPart=true, speedY=1.6}) {samplerType=SAMPLERTYPE_LinearGrayscale, texture=/Game/Effects/Fort_Effects/Emotes/UltraLight/Texture/T_Cloud_Greyscale}.R) {name=MainNoise, nodeColor=<rgba 0.6641 1 0 1>}
  Panner()
  ScalarParameter() {defaultValue=1, parameterName=ElectricityBrightness}
 ]
}
