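# Anim Blueprint for the BariumDemon pickaxe that offsets and scales nine front/back blade bone pairs, apparently for a staggered expand on equip.
# EventGraph derives is_melee from the player's weapon animation type, holding nine equip alphas at 1 otherwise, then zeroing them after staggered delays on melee.
# AnimGraph chains interpolated additive bone transforms, bypassed via cached pose in the front end.
# The file shows no binding from those alphas to the transforms.
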AnimBlueprint /BRCosmetics/Weapons/FORT_Melee/Pickaxe_BariumDemon/Meshes/BariumDemon_Axe_AnimBP
target_skeleton=&/BariumDemon_Axe_Skeleton
graphs [
 AnimGraph {nodes [
   n1 = "Transform (Modify) Bone - Bone: blade_01_front"(
    ComponentPose=Local To Component(LocalPose=Use cached pose 'pre_expand_cache'())
    Alpha=1
    Translation="-5,0,0"
    Rotation="0,0,0"
    Scale="0.99,1,1"
   ) {
    scaleMode=BMM_Additive
    scaleSpace=BCS_BoneSpace
    translationMode=BMM_Additive
    translationSpace=BCS_BoneSpace
    alphaBoolBlend {blendOption=Linear}
    alphaScaleBiasClamp {bInterpResult=true, interpSpeedDecreasing=7.5, interpSpeedIncreasing=7.5}
    boneToModify {boneName=blade_01_front}
    scale {x=0.99}
    translation {x=-5}
   }
   n2 = "Transform (Modify) Bone - Bone: blade_01_back"(ComponentPose=@n1, Alpha=1, Translation="-5,0,0", Rotation="0,0,0", Scale="0.99,1,1") {
    scaleMode=BMM_Additive
    scaleSpace=BCS_BoneSpace
    translationMode=BMM_Additive
    translationSpace=BCS_BoneSpace
    alphaBoolBlend {blendOption=Linear}
    alphaScaleBiasClamp {bInterpResult=true, interpSpeedDecreasing=7.5, interpSpeedIncreasing=7.5}
    boneToModify {boneName=blade_01_back}
    scale {x=0.99}
    translation {x=-5}
   }
   n3 = "Transform (Modify) Bone - Bone: blade_02_front"(ComponentPose=@n2, Alpha=1, Translation="-2.5,0,0", Rotation="0,0,0", Scale="0.99,1,1") {
    scaleMode=BMM_Additive
    scaleSpace=BCS_BoneSpace
    translationMode=BMM_Additive
    translationSpace=BCS_BoneSpace
    alphaBoolBlend {blendOption=Linear}
    alphaScaleBiasClamp {bInterpResult=true, interpSpeedDecreasing=7.5, interpSpeedIncreasing=7.5}
    boneToModify {boneName=blade_02_front}
    scale {x=0.99}
    translation {x=-2.5}
   }
   n4 = "Transform (Modify) Bone - Bone: blade_02_back"(ComponentPose=@n3, Alpha=1, Translation="-2.5,0,0", Rotation="0,0,0", Scale="0.99,1,1") {
    scaleMode=BMM_Additive
    scaleSpace=BCS_BoneSpace
    translationMode=BMM_Additive
    translationSpace=BCS_BoneSpace
    alphaBoolBlend {blendOption=Linear}
    alphaScaleBiasClamp {bInterpResult=true, interpSpeedDecreasing=7.5, interpSpeedIncreasing=7.5}
    boneToModify {boneName=blade_02_back}
    scale {x=0.99}
    translation {x=-2.5}
   }
   n5 = "Transform (Modify) Bone - Bone: blade_03_front"(ComponentPose=@n4, Alpha=1, Translation="-2,0,0", Rotation="0,0,0", Scale="0.99,1,1") {
    scaleMode=BMM_Additive
    scaleSpace=BCS_BoneSpace
    translationMode=BMM_Additive
    translationSpace=BCS_BoneSpace
    alphaBoolBlend {blendOption=Linear}
    alphaScaleBiasClamp {bInterpResult=true, interpSpeedDecreasing=7.5, interpSpeedIncreasing=7.5}
    boneToModify {boneName=blade_03_front}
    scale {x=0.99}
    translation {x=-2}
   }
   n6 = "Transform (Modify) Bone - Bone: blade_03_back"(ComponentPose=@n5, Alpha=1, Translation="-2,0,0", Rotation="0,0,0", Scale="0.9,1,1") {
    scaleMode=BMM_Additive
    scaleSpace=BCS_BoneSpace
    translationMode=BMM_Additive
    translationSpace=BCS_BoneSpace
    alphaBoolBlend {blendOption=Linear}
    alphaScaleBiasClamp {bInterpResult=true, interpSpeedDecreasing=7.5, interpSpeedIncreasing=7.5}
    boneToModify {boneName=blade_03_back}
    scale {x=0.9}
    translation {x=-2}
   }
   n7 = "Transform (Modify) Bone - Bone: blade_04_front"(ComponentPose=@n6, Alpha=1, Translation="-1.5,0,0", Rotation="0,0,0", Scale="0.9,1,1") {
    scaleMode=BMM_Additive
    scaleSpace=BCS_BoneSpace
    translationMode=BMM_Additive
    translationSpace=BCS_BoneSpace
    alphaBoolBlend {blendOption=Linear}
    alphaScaleBiasClamp {bInterpResult=true, interpSpeedDecreasing=7.5, interpSpeedIncreasing=7.5}
    boneToModify {boneName=blade_04_front}
    scale {x=0.9}
    translation {x=-1.5}
   }
   n8 = "Transform (Modify) Bone - Bone: blade_04_back"(ComponentPose=@n7, Alpha=1, Translation="-1.5,0,0", Rotation="0,0,0", Scale="0.9,1,1") {
    scaleMode=BMM_Additive
    scaleSpace=BCS_BoneSpace
    translationMode=BMM_Additive
    translationSpace=BCS_BoneSpace
    alphaBoolBlend {blendOption=Linear}
    alphaScaleBiasClamp {bInterpResult=true, interpSpeedDecreasing=7.5, interpSpeedIncreasing=7.5}
    boneToModify {boneName=blade_04_back}
    scale {x=0.9}
    translation {x=-1.5}
   }
   n9 = "Transform (Modify) Bone - Bone: blade_05_front"(ComponentPose=@n8, Alpha=1, Translation="-1,0,0", Rotation="0,0,0", Scale="0.9,1,1") {
    scaleMode=BMM_Additive
    scaleSpace=BCS_BoneSpace
    translationMode=BMM_Additive
    translationSpace=BCS_BoneSpace
    alphaBoolBlend {blendOption=Linear}
    alphaScaleBiasClamp {bInterpResult=true, interpSpeedDecreasing=7.5, interpSpeedIncreasing=7.5}
    boneToModify {boneName=blade_05_front}
    scale {x=0.9}
    translation {x=-1}
   }
   n10 = "Transform (Modify) Bone - Bone: blade_05_back"(ComponentPose=@n9, Alpha=1, Translation="-1,0,0", Rotation="0,0,0", Scale="0.9,1,1") {
    scaleMode=BMM_Additive
    scaleSpace=BCS_BoneSpace
    translationMode=BMM_Additive
    translationSpace=BCS_BoneSpace
    alphaBoolBlend {blendOption=Linear}
    alphaScaleBiasClamp {bInterpResult=true, interpSpeedDecreasing=7.5, interpSpeedIncreasing=7.5}
    boneToModify {boneName=blade_05_back}
    scale {x=0.9}
    translation {x=-1}
   }
   n11 = "Transform (Modify) Bone - Bone: blade_06_front"(ComponentPose=@n10, Alpha=1, Translation="-1,0,0", Rotation="0,0,0", Scale="0.9,1,1") {
    scaleMode=BMM_Additive
    scaleSpace=BCS_BoneSpace
    translationMode=BMM_Additive
    translationSpace=BCS_BoneSpace
    alphaBoolBlend {blendOption=Linear}
    alphaScaleBiasClamp {bInterpResult=true, interpSpeedDecreasing=7.5, interpSpeedIncreasing=7.5}
    boneToModify {boneName=blade_06_front}
    scale {x=0.9}
    translation {x=-1}
   }
   n12 = "Transform (Modify) Bone - Bone: blade_06_back"(ComponentPose=@n11, Alpha=1, Translation="-1,0,0", Rotation="0,0,0", Scale="0.9,1,1") {
    scaleMode=BMM_Additive
    scaleSpace=BCS_BoneSpace
    translationMode=BMM_Additive
    translationSpace=BCS_BoneSpace
    alphaBoolBlend {blendOption=Linear}
    alphaScaleBiasClamp {bInterpResult=true, interpSpeedDecreasing=7.5, interpSpeedIncreasing=7.5}
    boneToModify {boneName=blade_06_back}
    scale {x=0.9}
    translation {x=-1}
   }
   n13 = "Transform (Modify) Bone - Bone: blade_07_front"(ComponentPose=@n12, Alpha=1, Translation="-1,0,0", Rotation="0,0,0", Scale="0.9,1,1") {
    scaleMode=BMM_Additive
    scaleSpace=BCS_BoneSpace
    translationMode=BMM_Additive
    translationSpace=BCS_BoneSpace
    alphaBoolBlend {blendOption=Linear}
    alphaScaleBiasClamp {bInterpResult=true, interpSpeedDecreasing=7.5, interpSpeedIncreasing=7.5}
    boneToModify {boneName=blade_07_front}
    scale {x=0.9}
    translation {x=-1}
   }
   n14 = "Transform (Modify) Bone - Bone: blade_07_back"(ComponentPose=@n13, Alpha=1, Translation="-1,0,0", Rotation="0,0,0", Scale="0.9,1,1") {
    scaleMode=BMM_Additive
    scaleSpace=BCS_BoneSpace
    translationMode=BMM_Additive
    translationSpace=BCS_BoneSpace
    alphaBoolBlend {blendOption=Linear}
    alphaScaleBiasClamp {bInterpResult=true, interpSpeedDecreasing=7.5, interpSpeedIncreasing=7.5}
    boneToModify {boneName=blade_07_back}
    scale {x=0.9}
    translation {x=-1}
   }
   n15 = "Transform (Modify) Bone - Bone: blade_08_front"(ComponentPose=@n14, Alpha=1, Translation="-1,0,0", Rotation="0,0,0", Scale="0.9,1,1") {
    scaleMode=BMM_Additive
    scaleSpace=BCS_BoneSpace
    translationMode=BMM_Additive
    translationSpace=BCS_BoneSpace
    alphaBoolBlend {blendOption=Linear}
    alphaScaleBiasClamp {bInterpResult=true, interpSpeedDecreasing=7.5, interpSpeedIncreasing=7.5}
    boneToModify {boneName=blade_08_front}
    scale {x=0.9}
    translation {x=-1}
   }
   n16 = "Transform (Modify) Bone - Bone: blade_08_back"(ComponentPose=@n15, Alpha=1, Translation="-1,0,0", Rotation="0,0,0", Scale="0.9,1,1") {
    scaleMode=BMM_Additive
    scaleSpace=BCS_BoneSpace
    translationMode=BMM_Additive
    translationSpace=BCS_BoneSpace
    alphaBoolBlend {blendOption=Linear}
    alphaScaleBiasClamp {bInterpResult=true, interpSpeedDecreasing=7.5, interpSpeedIncreasing=7.5}
    boneToModify {boneName=blade_08_back}
    scale {x=0.9}
    translation {x=-1}
   }
   n17 = "Transform (Modify) Bone - Bone: blade_09_front"(ComponentPose=@n16, Alpha=1, Translation="-1,0,0", Rotation="0,0,0", Scale="0.5,1,1") {
    scaleMode=BMM_Additive
    translationMode=BMM_Additive
    translationSpace=BCS_BoneSpace
    alphaBoolBlend {blendOption=Linear}
    alphaScaleBiasClamp {bInterpResult=true, interpSpeedDecreasing=7.5, interpSpeedIncreasing=7.5}
    boneToModify {boneName=blade_09_front}
    scale {x=0.5}
    translation {x=-1}
   }
   n18 = "Transform (Modify) Bone - Bone: blade_09_back"(ComponentPose=@n17, Alpha=1, Translation="-1,0,0", Rotation="0,0,0", Scale="0.5,1,1") {
    scaleMode=BMM_Additive
    translationMode=BMM_Additive
    translationSpace=BCS_BoneSpace
    alphaBoolBlend {blendOption=Linear}
    alphaScaleBiasClamp {bInterpResult=true, interpSpeedDecreasing=7.5, interpSpeedIncreasing=7.5}
    boneToModify {boneName=blade_09_back}
    scale {x=0.5}
    translation {x=-1}
   }
   Output Pose(Result=Blend Poses by bool(BlendPose_0=Use cached pose 'pre_expand_cache'(), BlendPose_1=Component To Local(ComponentPose=@n18), bActiveValue=Get bIsInFrontEnd()) {blendType=Linear}) {layerGroup=None}
   Save cached pose 'pre_expand_cache'(Pose=Input Pose()) {cacheName=pre_expand_cache}
  ]}
 EventGraph {nodes [
   Event BlueprintUpdateAnimation()
   > n1 = Sequence()
   > n2 = Cast To FortPlayerAnimInstance(Object=GetAnimInstance(self=GetOwnerSkeletalMeshComponent()))
   > n3 = Switch on EFortWeaponCoreAnimation!(execute, Selection:EFortWeaponCoreAnimation Enum=Get WeaponCoreAnim(self=@n2)) -> (
    Melee:Exec
    Pistol:Exec
    Shotgun:Exec
    PaperBlueprint:Exec
    Rifle:Exec
    MeleeOneHand:Exec
    MachinePistol:Exec
    RocketLauncher:Exec
    GrenadeLauncher:Exec
    GoingCommando:Exec
    AssaultRifle:Exec
    TacticalShotgun:Exec
    SniperRifle:Exec
    TrapPlacement:Exec
    ShoulderLauncher:Exec
    AbilityDecoTool:Exec
    Crossbow:Exec
    C4:Exec
    RemoteControl:Exec
    DualWield:Exec
    AR_BullPup:Exec
    AR_ForwardGrip:Exec
    MedPackPaddles:Exec
    SMG_P90:Exec
    AR_DrumGun:Exec
    Consumable_Small:Exec
    Consumable_Large:Exec
    Balloon:Exec
    MountedTurret:Exec
    CreativeTool:Exec
    ExplosiveBow:Exec
    AshtonIndigo:Exec
    AshtonChicago:Exec
    MeleeDualWield:Exec
    Unarmed:Exec
   )
   Set is_melee(
    execute=[
     @n3.AR_BullPup
     @n3.AR_DrumGun
     @n3.AR_ForwardGrip
     @n3.AbilityDecoTool
     @n3.AshtonChicago
     @n3.AshtonIndigo
     @n3.AssaultRifle
     @n3.Balloon
     @n3.C4
     @n3.Consumable_Large
     @n3.Consumable_Small
     @n3.CreativeTool
     @n3.Crossbow
     @n3.DualWield
     @n3.ExplosiveBow
     @n3.GoingCommando
     @n3.GrenadeLauncher
     @n3.MachinePistol
     @n3.MedPackPaddles
     @n3.MountedTurret
     @n3.PaperBlueprint
     @n3.Pistol
     @n3.RemoteControl
     @n3.Rifle
     @n3.RocketLauncher
     @n3.SMG_P90
     @n3.Shotgun
     @n3.ShoulderLauncher
     @n3.SniperRifle
     @n3.TacticalShotgun
     @n3.TrapPlacement
    ]
    is_melee:bool=false
   ) -> (Output_Get:bool)
   Set is_melee(execute=[@n3.Melee, @n3.MeleeDualWield, @n3.MeleeOneHand], is_melee=true)
   n4 = Sequence!(execute=@n1.then_1) -> (then_0, then_1, then_2:Exec, then_3:Exec, then_4:Exec, then_5:Exec, then_6:Exec, then_7:Exec, then_8:Exec)
   > n5 = Branch(Condition=Get is_melee() -> (is_melee:bool))
   > Do Once(Reset=@n6.then)
   > Set equip_01_alpha(equip_01_alpha:double=1) -> (Output_Get:double)
   > Delay(Duration=0.1)
   > Set equip_01_alpha(equip_01_alpha=0)
   n6 = Set equip_01_alpha(execute=@n5.else, equip_01_alpha=1)
   n7 = Branch(execute=@n4.then_3, Condition=Get is_melee())
   > Do Once(Reset=@n8.then)
   > Set equip_04_alpha(equip_04_alpha:double=1) -> (Output_Get:double)
   > Delay(Duration=0.13)
   > Set equip_04_alpha(equip_04_alpha=0)
   n8 = Set equip_04_alpha(execute=@n7.else, equip_04_alpha=1)
   n9 = Branch(execute=@n4.then_2, Condition=Get is_melee())
   > Do Once(Reset=@n10.then)
   > Set equip_03_alpha(equip_03_alpha:double=1) -> (Output_Get:double)
   > Delay(Duration=0.12)
   > Set equip_03_alpha(equip_03_alpha=0)
   n10 = Set equip_03_alpha(execute=@n9.else, equip_03_alpha=1)
   n11 = Branch(execute=@n4.then_1, Condition=Get is_melee())
   > Do Once(Reset=@n12.then)
   > Set equip_02_alpha(equip_02_alpha:double=1) -> (Output_Get:double)
   > Delay(Duration=0.11)
   > Set equip_02_alpha(equip_02_alpha=0)
   n12 = Set equip_02_alpha(execute=@n11.else, equip_02_alpha=1)
   n13 = Branch(execute=@n4.then_8, Condition=Get is_melee())
   > Do Once(Reset=@n14.then)
   > Set equip_09_alpha(equip_09_alpha:double=1) -> (Output_Get:double)
   > Delay(Duration=0.18)
   > Set equip_09_alpha(equip_09_alpha=0)
   n14 = Set equip_09_alpha(execute=@n13.else, equip_09_alpha=1)
   n15 = Branch(execute=@n4.then_6, Condition=Get is_melee())
   > Do Once(Reset=@n16.then)
   > Set equip_07_alpha(equip_07_alpha:double=1) -> (Output_Get:double)
   > Delay(Duration=0.16)
   > Set equip_07_alpha(equip_07_alpha=0)
   n16 = Set equip_07_alpha(execute=@n15.else, equip_07_alpha=1)
   n17 = Branch(execute=@n4.then_7, Condition=Get is_melee())
   > Do Once(Reset=@n18.then)
   > Set equip_08_alpha(equip_08_alpha:double=1) -> (Output_Get:double)
   > Delay(Duration=0.17)
   > Set equip_08_alpha(equip_08_alpha=0)
   n18 = Set equip_08_alpha(execute=@n17.else, equip_08_alpha=1)
   n19 = Branch(execute=@n4.then_5, Condition=Get is_melee())
   > Do Once(Reset=@n20.then)
   > Set equip_06_alpha(equip_06_alpha:double=1) -> (Output_Get:double)
   > Delay(Duration=0.15)
   > Set equip_06_alpha(equip_06_alpha=0)
   n20 = Set equip_06_alpha(execute=@n19.else, equip_06_alpha=1)
   n21 = Branch(execute=@n4.then_4, Condition=Get is_melee())
   > Do Once(Reset=@n22.then)
   > Set equip_05_alpha(equip_05_alpha:double=1) -> (Output_Get:double)
   > Delay(Duration=0.14)
   > Set equip_05_alpha(equip_05_alpha=0)
   n22 = Set equip_05_alpha(execute=@n21.else, equip_05_alpha=1)
  ]}
]
variables [
 equip_01_alpha {type=number}
 equip_02_alpha {type=number}
 equip_03_alpha {type=number}
 equip_04_alpha {type=number}
 equip_05_alpha {type=number}
 equip_06_alpha {type=number}
 equip_07_alpha {type=number}
 equip_08_alpha {type=number}
 equip_09_alpha {type=number}
 is_melee {type=boolean}
 scale_value {type=Vector, default=<0.95 1 1>}
]
